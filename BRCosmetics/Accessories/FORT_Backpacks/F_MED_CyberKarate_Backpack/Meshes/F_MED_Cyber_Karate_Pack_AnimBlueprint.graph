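# Animation Blueprint for the Cyber Karate backpack cosmetic, procedurally adjusting its pack bone to the wearer's pose.
# Each update reads the owner mesh's spine delta from reference pose to set roll, spine-to-left-clavicle delta during melee for swing alpha, and skydive floating for float alpha.
# The AnimGraph stacks three additive bone-space Modify Bone nodes on the input pose: spine-driven roll, melee swing offset and interpolated float offset.
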AnimBlueprint /BRCosmetics/Accessories/FORT_Backpacks/F_MED_CyberKarate_Backpack/Meshes/F_MED_Cyber_Karate_Pack_AnimBlueprint
target_skeleton=&/F_MED_Cyber_Karate_Pack_Skeleton
graphs [
 AnimGraph {nodes [
   n1 = "Transform (Modify) Bone - Bone: pack"(ComponentPose=Local To Component(LocalPose=Input Pose()), Translation="0,1,0", Rotation=Get backpack_rotate()) {
    rotationMode=BMM_Additive
    rotationSpace=BCS_BoneSpace
    translationMode=BMM_Additive
    translationSpace=BCS_BoneSpace
    alphaBoolBlend {blendOption=Linear}
    boneToModify {boneName=pack}
    translation {y=1}
   }
   n2 = "Transform (Modify) Bone - Bone: pack"(ComponentPose=@n1, Alpha=Get swing_rotate_alpha() -> (swing_rotate_alpha:double), Translation="0,-3.5,0", Rotation="30,0.000,10") {
    rotationMode=BMM_Additive
    rotationSpace=BCS_BoneSpace
    translationMode=BMM_Additive
    translationSpace=BCS_BoneSpace
    alphaBoolBlend {blendOption=Linear}
    alphaScaleBiasClamp {interpSpeedDecreasing=6, interpSpeedIncreasing=8}
    boneToModify {boneName=pack}
    rotation {pitch=30, roll=10}
    translation {y=-3.5}
   }
   n3 = "Transform (Modify) Bone - Bone: pack"(ComponentPose=@n2, Alpha=Get float_alpha() -> (float_alpha:double), Translation="0,-3,0", Rotation="0,0.000,-10") {
    rotationMode=BMM_Additive
    rotationSpace=BCS_BoneSpace
    translationMode=BMM_Additive
    translationSpace=BCS_BoneSpace
    alphaBoolBlend {blendOption=Linear}
    alphaScaleBiasClamp {bInterpResult=true, interpSpeedDecreasing=5}
    boneToModify {boneName=pack}
    rotation {roll=-10}
    translation {y=-3}
   }
   Output Pose(Result=Component To Local(ComponentPose=@n3)) {layerGroup=None}
  ]}
 EventGraph {nodes [
   Event BlueprintUpdateAnimation()
   n1 = GetOwnerSkeletalMeshComponent()
   n2 = BreakRotator(InRot=BreakTransform(InTransform=GetDeltaTransformFromRefPose(self=@n1, BoneName=spine_05, BaseName=spine_01)).Rotation)
   > Set backpack_rotate(backpack_rotate=MakeRotator(Roll=MapRangeClamped(Value=@n2.Yaw, InRangeA=0, InRangeB=-30, OutRangeB=20)))
   > Set float_alpha(float_alpha:double="To Float (Boolean)"(InBool=Get bIsSkydiveFloating())) -> (Output_Get:double)
   > n3 = Branch(Condition=Get bIsPlayingMeleeAnim())
   n4 = BreakTransform(InTransform=GetDeltaTransformFromRefPose(self=GetOwnerSkeletalMeshComponent(), BoneName=spine_05, BaseName=clavicle_l))
   > Set swing_rotate_alpha(swing_rotate_alpha:double=MapRangeClamped(Value=BreakRotator(InRot=@n4.Rotation).Yaw, InRangeA=10, InRangeB=40)) -> (Output_Get:double)
   Set swing_rotate_alpha(execute=@n3.else, swing_rotate_alpha=0)
   TryGetPawnOwner()
   Get swing_rotate_alpha()
  ]}
]
variables [
 backpack_rotate {type=Rotator}
 float_alpha {type=number}
 swing_rotate_alpha {type=number}
]
